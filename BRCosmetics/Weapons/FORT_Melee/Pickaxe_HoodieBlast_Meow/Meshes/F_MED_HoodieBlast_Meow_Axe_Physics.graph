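# Physics asset for the HoodieBlast Meow two-handed pickaxe, providing bodies and joints for its two short dangling chains.
# Capsules cover the handle and root bone, while small spheres cover the chain bones.
# The handle and both chain bases have angularly locked joints to the root.
# The two-link first chain has tight angle values except one wide swing at its outer joint; the single-link second chain has moderate swing and narrow twist values.
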PhysicsAsset /BRCosmetics/Weapons/FORT_Melee/Pickaxe_HoodieBlast_Meow/Meshes/F_MED_HoodieBlast_Meow_Axe_Physics
bodies {
 Melee_TwoHanded_Handle [{shapeType=Capsule, length=18.1222, radius=5.9576, center=<0.0241 -8.8301 73.7601>, rotation=<rot -34.334 -80.7627 -5.9346>}]
 b_AR_Root [{shapeType=Capsule, length=49.0419, radius=5.8314, center=<7.8257 0.012 7.1192>, rotation=<rot 0.0042 -90 1.4303>}]
 dyn_chain_01_01 [{shapeType=Sphere, radius=3.3552, shapeName=dyn_chain_01_01_sphere, center=<3.2892 0.0842 0.0124>}]
 dyn_chain_01_02 [{shapeType=Sphere, radius=5.7174, shapeName=dyn_chain_01_02_sphere, center=<2.4139 0.0119 -0.0014>}]
 dyn_chain_01_base [{shapeType=Sphere, radius=0.505, shapeName=dyn_chain_01_base_sphere}]
 dyn_chain_02_01 [{shapeType=Sphere, radius=4.9573, shapeName=dyn_chain_02_01_sphere, center=<4.7161 -0.1534 0.2738>}]
 dyn_chain_02_base1 [{shapeType=Sphere, radius=0.505, shapeName=dyn_chain_02_base1_sphere}]
}
constraints [
 {
  bone1Name=Melee_TwoHanded_Handle
  bone2Name=b_AR_Root
  swing1LimitDegrees=45
  swing1Motion=Locked
  swing2LimitDegrees=45
  swing2Motion=Locked
  twistLimitDegrees=45
  twistMotion=Locked
 }
 {
  bone1Name=dyn_chain_01_base
  bone2Name=b_AR_Root
  swing1LimitDegrees=45
  swing1Motion=Locked
  swing2LimitDegrees=45
  swing2Motion=Locked
  twistLimitDegrees=45
  twistMotion=Locked
 }
 {bone1Name=dyn_chain_01_01, bone2Name=dyn_chain_01_base, swing1LimitDegrees=3, swing2LimitDegrees=3, twistLimitDegrees=2}
 {bone1Name=dyn_chain_01_02, bone2Name=dyn_chain_01_01, swing1LimitDegrees=3, swing2LimitDegrees=117, twistLimitDegrees=2}
 {bone1Name=dyn_chain_02_01, bone2Name=dyn_chain_02_base1, swing1LimitDegrees=37.8, swing2LimitDegrees=45, twistLimitDegrees=4}
 {
  bone1Name=dyn_chain_02_base1
  bone2Name=b_AR_Root
  swing1LimitDegrees=45
  swing1Motion=Locked
  swing2LimitDegrees=45
  swing2Motion=Locked
  twistLimitDegrees=45
  twistMotion=Locked
 }
]
