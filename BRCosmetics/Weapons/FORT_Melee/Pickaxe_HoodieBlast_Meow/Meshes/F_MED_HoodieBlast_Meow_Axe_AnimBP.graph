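# Animation Blueprint for the HoodieBlast Meow pickaxe that adds secondary physics motion to the axe.
# The input pose runs through a RigidBody node simulated in b_AR_Root base-bone space, then a spring-driven Anim Dynamics node on dyn_spanner_base.
# Each update picks one of three sim-space and gravity presets from bIsInFrontEnd and, within the front end, bIsOnPawn, feeding the RigidBody node.
# The Transform Bone node is left unconnected.
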AnimBlueprint /BRCosmetics/Weapons/FORT_Melee/Pickaxe_HoodieBlast_Meow/Meshes/F_MED_HoodieBlast_Meow_Axe_AnimBP
target_skeleton=&/F_MED_HoodieBlast_Meow_Axe_Skeleton
graphs [
 AnimGraph {nodes [
   "Transform (Modify) Bone - Bone: b_AR_Root"(Alpha=1, Translation="0,7,-33", Rotation="0,0,0", Scale="1,1,1") {translationMode=BMM_Additive, boneToModify {boneName=b_AR_Root}, translation {y=7, z=-33}}
   AnimGraphNode_SubInput_0 = Input Pose()
   n1 = RigidBody!(
    ComponentPose=Local To Component(LocalPose=@AnimGraphNode_SubInput_0)
    Alpha
    OverrideWorldGravity:Vector=Get Gravity() -> (Gravity:Vector)
    ExternalForce
    SimSpaceSettings:Sim Space Settings Structure=Get SimSpaceSettings() -> (SimSpaceSettings:Sim Space Settings Structure)
   ) -> (Pose) {
    bOverrideWorldGravity=true
    overridePhysicsAsset=&/F_MED_HoodieBlast_Meow_Axe_Physics
    simulationSpace=BaseBoneSpace
    baseBoneRef {boneName=b_AR_Root}
    overrideWorldGravity {z=-1015}
    simSpaceSettings {
     dampingAlpha=0.4
     maxAngularAcceleration=225
     maxAngularVelocity=45
     maxLinearAcceleration=1500
     maxLinearVelocity=250
     worldAlpha=1
     externalLinearDragV=<0.9 0.9 0.9>
    }
   }
   n2 = Anim Dynamics - Bone: dyn_spanner_base(ComponentPose=@n1, Alpha=1, ExternalForce="0,0,0") {
    angularDampingOverride=0.99
    angularSpringConstant=200
    bAngularSpring=true
    bLinearSpring=true
    bOverrideAngularDamping=true
    bOverrideLinearDamping=true
    bUseGravityOverride=true
    linearDampingOverride=1
    linearSpringConstant=50
    boundBone {boneName=dyn_spanner_base}
    componentAppliedLinearAccClamp=<1000 1000 1000>
    componentLinearAccScale=<0.3 0.3 0.3>
    componentLinearVelScale=<0.3 0.3 0.3>
    gravityOverride {z=-980}
    physicsBodyDefinitions [{
      boundBone {boneName=dyn_spanner_base}
      boxExtents=<11 11 11>
      constraintSetup {angularLimitsMax=<0 0 2>, angularLimitsMin=<0 0 -2>, angularTarget=<0 0 0>, linearAxesMax=<2 0 0>, linearAxesMin=<-5 0 0>}
      localJointOffset=<1 0 0>
     }]
   }
   Output Pose(Result=Component To Local(ComponentPose=@n2)) {layerGroup=None}
  ]}
 EventGraph {nodes [
   Event BlueprintUpdateAnimation()
   > n1 = Branch(Condition=Get bIsInFrontEnd())
   > n2 = Branch(Condition=Get bIsOnPawn() -> (bIsOnPawn:bool))
   > Set SimSpaceSettings(SimSpaceSettings:Sim Space Settings Structure=Get SimSpaceSettings_2() -> (SimSpaceSettings_2:Sim Space Settings Structure)) -> (Output_Get:Sim Space Settings Structure)
   > Set Gravity(Gravity:Vector="0,0,-1980") -> (Output_Get:Vector)
   Set SimSpaceSettings(execute=@n2.else, SimSpaceSettings=Get SimSpaceSettings_0() -> (SimSpaceSettings_0:Sim Space Settings Structure))
   > Set Gravity(Gravity="0,0,-980")
   Set SimSpaceSettings(execute=@n1.else, SimSpaceSettings=Get SimSpaceSettings_1() -> (SimSpaceSettings_1:Sim Space Settings Structure))
   > Set Gravity(Gravity="0,0,-1015")
   TryGetPawnOwner()
  ]}
]
variables [
 Gravity {type=Vector}
 Sim Space Settings {
  type=AnimPhysSimSpaceSettings
  default {maxAngularAcceleration=10000, maxAngularVelocity=10000, simSpaceAngularAlpha=0, externalAngularVelocity=<0 0 0>}
 }
 SimSpaceSettings {
  type=SimSpaceSettings
  default {
   dampingAlpha=1
   maxAngularAcceleration=10000
   maxAngularVelocity=10000
   maxLinearAcceleration=10000
   maxLinearVelocity=10000
   velocityScaleZ=1
   worldAlpha=0
   externalLinearDragV=<0 0 0>
  }
 }
 SimSpaceSettings_0 {
  type=SimSpaceSettings
  default {
   dampingAlpha=0
   maxAngularAcceleration=70
   maxAngularVelocity=7
   maxLinearAcceleration=800
   maxLinearVelocity=200
   velocityScaleZ=0.2
   worldAlpha=1
   externalLinearDragV=<0 0 0>
  }
 }
 SimSpaceSettings_1 {
  type=SimSpaceSettings
  default {
   dampingAlpha=0
   maxAngularAcceleration=20
   maxAngularVelocity=8
   maxLinearAcceleration=1300
   maxLinearVelocity=30
   velocityScaleZ=0.7
   worldAlpha=1
   externalLinearDragV=<1 0 0>
  }
 }
 SimSpaceSettings_2 {
  type=SimSpaceSettings
  default {
   dampingAlpha=0.1
   maxAngularAcceleration=48
   maxAngularVelocity=4.8
   maxLinearAcceleration=785
   maxLinearVelocity=200
   velocityScaleZ=0.2
   worldAlpha=1
   externalLinearDragV=<0 0 0>
  }
 }
]
